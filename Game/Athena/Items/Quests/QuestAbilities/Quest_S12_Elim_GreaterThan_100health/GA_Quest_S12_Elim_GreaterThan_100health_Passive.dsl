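; Passive gameplay ability for a Season 12 quest that marks the owning pawn while its health plus shield is at least 100, apparently for elimination tracking.
; On activation, and on each pawn health-changed event, HealthCheck feeds UpdateHealthGE.
; That function applies the quest GameplayEffect or removes effects granting the Athena.Quests.Elim.Greaterthan100health tag.
; Application has no check for an already active effect.
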(event Ability|EventActivateAbility
  (|CallParentFunction)
  ;[1 1 1] On Spawn Check Health and Update GE
  (bind _health_greater_than_100 (CallFunction|HealthCheck))
  (CallFunction|UpdateHealthGE _health_greater_than_100)
  (Health|BindEventtoOnPawnHealthChanged (Variables|GAVars|Data|GetOwningPlayerPawn) (delegate Custom|Health%20Changed)))

(event Custom|Health%20Changed
  ;[1 1 1] On Health Changed Check Health and Update GE
  (bind _health_greater_than_100 (CallFunction|HealthCheck))
  (CallFunction|UpdateHealthGE _health_greater_than_100))

(event Ability|EventFailedToActivatePassiveAbility
  (params (ActorInfo :type Gameplay%20Ability%20Actor%20Info%20Structure))
  (|CallParentFunction))

(fn HealthCheck
  (return (>= (+ (FortPawn|GetHealth (Variables|GAVars|Data|GetOwningPlayerPawn)) (FortPawn|GetShield (Variables|GAVars|Data|GetOwningPlayerPawn))) 1e+02)))

(fn UpdateHealthGE
  (params (Health%20Greater%20Than%20100 :type Boolean))
  (if Health%20Greater%20Than%20100
    (GameplayEffects|ApplyGameplayEffectToSelf :self (Ability|GetAbilitySystemComponent (Variables|GAVars|Data|GetOwningPlayerPawn)) :GameplayEffectClass "/Game/Athena/Items/Quests/QuestAbilities/Quest_S12_Elim_GreaterThan_100health/GE_Quest_S12_Elim_GreaterThan_100health.GE_Quest_S12_Elim_GreaterThan_100health_C" :Level 0.0)
    (else
      (GameplayEffects|RemoveActiveEffectswithGrantedTags (Ability|GetAbilitySystemComponent (Variables|GAVars|Data|GetOwningPlayerPawn)) "(GameplayTags=((TagName=\"Athena.Quests.Elim.Greaterthan100health\")))"))))
